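# Niagara system for an ambient environment bird flock, drawing a burst of roughly a dozen to twenty mesh particles that circle overhead or nearby.
# Spawn randomizes scale, tint and four dynamic material parameters per bird for variation.
# RotateAroundPoint gives randomized radius, center, phase and rate, and its forward vector is written to velocity, alongside an initial mesh orientation.
# Birds ignore lifetime expiry, and infinitely lived ones die when the emitter deactivates.
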NiagaraSystem /Game/Effects/Fort_Effects/Effects/Environment/Birds/NS_BirdFlock_Orbit
dependencies {usedDataInterfaces [], usedDynamicInputs [], usedModules [&SystemState]}
emitters {Particle_Emitter {
  emitterUpdateScript [
   &EmitterState(Life Cycle Mode=::NewEnumerator1, Loop Behavior=::NewEnumerator1, Loop Duration Mode=::NewEnumerator1, UseLoopDelay=true)
   &SpawnBurst_Instantaneous(
    Spawn Count=&RandomRangeInt(Minimum=12, Maximum=20)
    Use Spawn Probability=false
    UseLoopCountLimit=false
   )
  ]
  particleSpawnScript [
   &InitializeParticle(
    Mesh Scale=&VectorFromFloat(Value=&RandomRangeFloat(Minimum=0.75, Maximum=1.2, Evaluation Type=::NewEnumerator1))
    Mesh Scale Mode=::NewEnumerator4
    UsePositionOffset=false
    Write Material Random=true
   )
   &Color(Color=&MakeLinearColorFromVectorAndFloat(
     "Vector (RGB)"=&RandomRangeVector(Minimum=<0.5 0.5 0.5>, Maximum=<1 1 1>, Evaluation Type=::NewEnumerator1)
     "Float (Alpha)"=1
    ))
   &DynamicMaterialParameters(
    Index 0 Param 1=&RandomRangeFloat(Minimum=3, Maximum=6, Evaluation Type=::NewEnumerator1)
    Index 0 Param 2=&RandomRangeFloat(Minimum=1, Maximum=1.5, Evaluation Type=::NewEnumerator1)
    Index 0 Param 3=&RandomRangeFloat(Minimum=-60, Maximum=60, Evaluation Type=::NewEnumerator1)
    Index 0 Param 4=&RandomRangeFloat(Minimum=1, Maximum=0.75, Evaluation Type=::NewEnumerator1)
    Module.Param0WriteEnabled=true
    Module.Param1WriteEnabled=true
    Module.Param2WriteEnabled=true
    Module.Param3WriteEnabled=true
   )
   &InitialMeshOrientation(Mesh Orientation Mode=::NewEnumerator7, Rotation=<0 0 0.5>, Use Rotation Vector=true)
  ]
  particleUpdateScript [
   &ParticleState(Kill Particles When Lifetime Has Elapsed=false, Let Infinitely Lived Particles Die When Emitter Deactivates=true)
   &ScaleColor(Scale RGB=<0.5 0.5 0.5>, ScaleA=true, ScaleRGB=true)
   &RotateAroundPoint(
    Radius:NiagaraFloat=&RandomRangeFloat(Minimum=1200, Maximum=2400)
    Rotation Center:NiagaraPosition=&RandomRangeVector(Minimum=<-500 -500 -500>, Maximum=<500 500 500>)
    Rotation Phase:NiagaraFloat=&RandomRangeFloat(Minimum=0, Maximum=1)
    Rotation Rate:NiagaraFloat=&RandomRangeFloat(Minimum=0.02, Maximum=0.06)
   )
   /SetVariables(Particles.Velocity=&Multiply_Vector(A=$Output.RotateAroundPoint.LocalForwardVector, B=<1 1 1>)) {isSetParameters=true}
  ]
  rendererClasses [&NiagaraMeshRendererProperties]
  renderers [{rendererClass=&NiagaraMeshRendererProperties, rendererIndex=0}]
 }}
summary {
 systemName=NS_BirdFlock_Orbit
 emitters [{emitterName=Particle_Emitter, rendererClasses [&NiagaraMeshRendererProperties]}]
 userVariables []
}
